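; Projectile Blueprint for the TwinStretchCorn weapon, exploding through its parent delivery logic.
; BeginPlay sets async explosions and, only if the owner weapon's spawned shield is valid, ignores that shield when moving, starts in-air audio tagged local or enemy, and sets a lifespan.
; The audio fades out once when the projectile stops.
; Entering a first water body records the surface location, flags the water hit and triggers the explosion.
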(event EventBeginPlay
  (bind _returnvalue (Game|GetInstigator))
  (Variables|Gameplay|SetbUseAsyncExplosions (ScalableFloat|GetValueasBool (Variables|Default|GetHF_UseAsyncExplosion)))
  (|CallParentFunction)
  (bind _returnvalue_2 (CallFunction|GetOwnerWeapon))
  (bind _asb_weap_twin_stretch_corn (Utilities|Casting|CastToB_Weap_TwinStretchCorn _returnvalue_2))
  (bind _output_get (Variables|Default|SetWeap%20Twin%20Stretch%20Corn _asb_weap_twin_stretch_corn))
  (bind _spawnedshield (Variables|/TwinStretchCorn/Gameplay/Actors/B_Weap_TwinStretchCorn.B_Weap_TwinStretchCorn_C|GetSpawnedShield _output_get)
    (:then
      (Collision|IgnoreActorwhenMoving (Variables|Collision|GetCapsuleComponent) _spawnedshield true)
      (CallFunction|GetOwnerWeapon)
      ;[0.116 0.3229 1] Play Audio InAir Loop
      (Audio|Components|Audio|FadeIn (Variables|Default|GetAudio_InAir_Loop) 0.1 1.0 0.0 "Linear" :!class /Script/Engine.AudioComponent)
      (Audio|Parameter|SetBooleanParameter (Variables|Default|GetAudio_InAir_Loop) "Islocal" (Pawn|IsLocallyViewed _returnvalue) :!class /Script/AudioExtensions.AudioParameterControllerInterface)
      (Audio|Parameter|SetBooleanParameter (Variables|Default|GetAudio_InAir_Loop) "IsEnemy" (Utilities|Enum|Equal(Enum) (Team|GetTeamAffiliationforActortoLocalPlayer _returnvalue true true) "Hostile") :!class /Script/AudioExtensions.AudioParameterControllerInterface)
      (Actor|SetLifeSpan 1e+01))
    (:else)))

(event Collision|EventHit
  (params (MyComp :type Primitive%20Component%20Object%20Reference) (Other :type Actor%20Object%20Reference) (OtherComp :type Primitive%20Component%20Object%20Reference) (bSelfMoved :type Boolean) (HitLocation :type Vector :default "0, 0, 0") (HitNormal :type Vector :default "0, 0, 0") (NormalImpulse :type Vector :default "0, 0, 0") (Hit :type Hit%20Result%20Structure%20%28by%20ref%29)))

(event OnProjectileStop(ProjectileMovementComponent)
  (params (ImpactResult :type Hit%20Result%20Structure%20%28by%20ref%29))
  (Utilities|FlowControl|DoOnce false
    (:Completed
      (Audio|Components|Audio|FadeOut (Variables|Default|GetAudio_InAir_Loop) 0.3 0.0 "Linear" :!class /Script/Engine.AudioComponent))))

(event EventOnTouched
  (params (OtherActor :type Actor%20Object%20Reference) (OtherComp :type Primitive%20Component%20Object%20Reference) (HitResult :type Hit%20Result%20Structure%20%28by%20ref%29) (bIsOverlap :type Boolean))
  (|CallParentFunction OtherActor OtherComp HitResult bIsOverlap))

(event OnEnterWaterEvent(WaterInteractionComponent)
  (params (WaterBody :type Fort%20Water%20Body%20Actor%20Object%20Reference) (WaterInteractionComponent :type Fort%20Water%20Interaction%20Component%20Object%20Reference) (bIsFirstBody :type Boolean))
  ;[1 1 1] Handle Water
  (if bIsFirstBody
    (bind _output_get (Variables|Default|SetWater%20Body WaterBody))
    (Variables|Default|SetWater%20Interaction%20Component WaterInteractionComponent)
    (Variables|Default|SetOut%20Water%20Surface%20Location (WaterBody|GetWaterSurfaceInfoatLocation (Water|GetWaterBodyComponent _output_get) (Transformation|GetWorldLocation (Variables|Collision|GetCapsuleComponent)) false))
    (Variables|Default|SetHitWater true)
    (Delivery|DoExplosion)))

(fn ConstructionScript)
