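; TwinStretchCorn weapon Blueprint whose magazine ammo serves as shards for both queued charged projectiles and a blocking shield.
; The authority spawns an attached block-volume shield; each depleted per-shard health pool costs one ammo.
; A gameplay tag count drives blocking state, while the queued-projectile rep-notify drives charged-shard effects.
; Niagara idle effects follow ammo and defense state, and unequipping clears the queue, removes the shield and skips fall damage if airborne.
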(event EventBeginPlay
  ;[1 1 1] BeginPlay
  (|CallParentFunction)
  (EventDispatchers|AssignOnLocalAmmoChanged (delegate Custom|OnLocalAmmoChanged_Event))
  (bind (_asynctaskproxy _tagcount) (Ability|Tasks|WaitGameplayTagCountChangedOnActor (Game|GetInstigator) (Variables|Shield|GetBlockingGrantedTag))
    (:then
      (Ability|Async|WaitGameplayEventToActor (Game|GetInstigator) "(TagName=\"Event.TwinStretchCorn.Shield.Hit\")"))
    (:TagCountChanged
      (CallFunction|HandleBlockingStateChanged (> _tagcount 0)))))

(event Custom|OnLocalAmmoChanged_Event
  (params (LocalCount :type Integer) (LocalRemaining :type Integer))
  ;[1 1 1] BeginPlay
  (Utilities|FlowControl|Sequence
    (:then_0
      (if (<= LocalCount 0)
        (Default|CallOnOutOfAmmo)
        (FX|UpdateIdleFX 0 false false false 0 true)
        (Audio|SpawnSoundAttached "/TwinStretchCorn/Assets/Audio/Cues/SC_TwinStretchCorn_NoAmmo.SC_TwinStretchCorn_NoAmmo" (Variables|Visual|GetWeaponMesh))
        (else
          (FX|SetIdleFXVisibility true))))))

(event FortWeapon|EventOnEquippedAndAttached
  ;[1 1 1] Equip
  (|CallParentFunction)
  (Utilities|FlowControl|Sequence
    (:then_0
      (Variables|Default|GetSpawnedIdleFX
        (:then)
        (:else
          (bind _returnvalue (Niagara|SpawnSystemAttached (Variables|Default|GetIdleFXTemplate) (Variables|/Script/Engine.Actor|GetRootComponent (Game|GetInstigator)) "None" "0, 0, 0" "0, 0, 0" "KeepRelativeOffset" false true "ManualRelease"))
          (Variables|Default|SetSpawnedIdleFX _returnvalue))))
    (:then_1
      (bind _niagara_fx (FX|SetIdleFXVisibility (> (FortWeapon|GetMagazineAmmoCount) 0)))
      (Effects|Components|ParticleSystem|SetIntParameter _niagara_fx "Ammo Count" (FortWeapon|GetMagazineAmmoCount))
      (FX|UpdateIdleFX (FortWeapon|GetMagazineAmmoCount) true)
      (Components|Activation|Activate _niagara_fx false))
    (:then_2
      (bind _spawnedshield (Variables|Shield|GetSpawnedShield)
        (:then
          (Class|BGATwinStretchCornBlockVolume|SetIdleFxVisible _spawnedshield true))
        (:else)))))

(event FortWeapon|EventOnUnEquippedAndDetached
  ;[1 1 1] Unequip
  (|CallParentFunction)
  (Charging|ClearQueuedProjectiles)
  (Utilities|FlowControl|Sequence
    (:then_0
      (Variables|Default|GetSpawnedIdleFX
        (:then
          (FX|UpdateIdleFX (FortWeapon|GetMagazineAmmoCount) false false false 0 true true))
        (:else)))
    (:then_1
      (bind _spawnedshield (Variables|Shield|GetSpawnedShield)
        (:then
          (Class|BGATwinStretchCornBlockVolume|SetIdleFxVisible _spawnedshield false))
        (:else)))
    (:then_2
      (bind _asfort_player_pawn (Utilities|Casting|CastToFortPlayerPawn (Game|GetInstigator)))
      (if (AI|Components|NavMovement|IsFalling (Variables|/Script/Engine.Character|GetCharacterMovement _asfort_player_pawn))
        (Variables|/Script/FortniteGame.FortPawn|SetbIgnoreNextFallingDamage true _asfort_player_pawn)))
    (:then_3
      (bind _spawnedshield_2 (Variables|Shield|GetSpawnedShield)
        (:then
          (Gameplay|SilentDie _spawnedshield_2 false :!class /Script/FortniteGame.BuildingActor)
          (bind _spawnedidlefx (Variables|Default|GetSpawnedIdleFX)
            (:then
              (Components|DestroyComponent _spawnedidlefx))
            (:else)))
        (:else)))))

(event Custom|Update%20Shield%20Position
  (bind _spawnedshield (Variables|Shield|GetSpawnedShield))
  (bind _returnvalue (Transformation|GetActorTransform _spawnedshield))
  (bind _spawnedidlefx (Variables|Default|GetSpawnedIdleFX))
  (Effects|Components|ParticleSystem|SetVectorParameter _spawnedidlefx "Shield Pos" (.location _returnvalue))
  (Effects|Components|ParticleSystem|SetVectorParameter _spawnedidlefx "Shield Rot" (Math|Rotator|GetRotationXVector (.rotation _returnvalue))))

(event Custom|HandleBlockingStateChanged
  (params (IsBlocking :type Boolean))
  (Default|CallOnBlockingStateChanged IsBlocking)
  (FX|UpdateIdleFX (FortWeapon|GetMagazineAmmoCount) false IsBlocking))

(fn ActivateOrSpawnShield
  (Networking|SwitchHasAuthority
    (:Authority
      (Variables|Shield|GetSpawnedShield
        (:then
          _)
        (:else
          (Shield|SpawnShield)))
      (Class|BGATwinStretchCornBlockVolume|SetShieldActive (Variables|Shield|GetSpawnedShield) true)
      (FX|UpdateIdleFX 0 false true)
      (return (Variables|Shield|GetSpawnedShield)))))

(fn AddQueuedProjectile
  (Variables|Charging|SetNumQueuedProjectiles (+ (Variables|Charging|GetNumQueuedProjectiles) 1))
  (GameplayCue|ExecuteGameplayCue (Game|GetInstigator) (Variables|Charging|GetShardChargedCue)))

(fn ClearQueuedProjectiles
  (Variables|Charging|SetNumQueuedProjectiles 0))

(fn DeactivateShield
  (bind _spawnedshield (Variables|Shield|GetSpawnedShield)
    (:then
      (Class|BGATwinStretchCornBlockVolume|SetShieldActive _spawnedshield false)
      (FX|UpdateIdleFX (FortWeapon|GetMagazineAmmoCount) true))
    (:else)))

(fn GetCurrentShardLocations
  (bind _queuedprojectilerelativelocations (Variables|Charging|GetQueuedProjectileRelativeLocations))
  (for _index (range (Variables|Charging|GetNumQueuedProjectiles))
    (Development|PrintString (Utilities|String|ToString(Vector) (Utilities|Array|Get(acopy) _queuedprojectilerelativelocations _index)))
    (Utilities|Array|Add (LocalVariables|GetOutLocations) (Utilities|Array|Get(acopy) _queuedprojectilerelativelocations _index)))
  (return (LocalVariables|GetOutLocations)))

(fn GetCurrentTotalShieldHealth
  (return (+ (Variables|Shield|GetShieldHealthRemaining) (* (ScalableFloat|GetValueatLevel (Variables|Shield|GetMaxShieldHealthPerShard)) (- (FortWeapon|GetMagazineAmmoCount) 1)))))

(fn GetMaxTotalShieldHealth
  (return (* (ScalableFloat|GetValueatLevel (Variables|Shield|GetMaxShieldHealthPerShard)) (FortWeapon|GetBulletsPerClip))))

(fn GetQueuedProjectileWorldLocation
  (params (Index :type Integer))
  (return (+ (Math|Vector|RotateVector (Utilities|Array|Get(acopy) (Variables|Charging|GetQueuedProjectileRelativeLocations) Index) (Transformation|GetActorRotation (Game|GetInstigator))) (Transformation|GetActorLocation (Game|GetInstigator)))))

(fn GetShieldActor
  (return (Variables|Shield|GetSpawnedShield)))

(fn GetShieldTransform
  (bind (_returnvalue_location _returnvalue_rotation _returnvalue_scale) (Math|Transform|BreakTransform (Transformation|GetActorTransform (Game|GetInstigator))))
  (return (+ _returnvalue_location (Math|Vector|RotateVector (Variables|Shield|GetShieldOffset) _returnvalue_rotation)) _returnvalue_rotation _returnvalue_scale))

(fn HandleShieldDamage
  (params (Damage :type Float%20%28double-precision%29))
  (bind _output_get (Variables|Shield|SetShieldHealthRemaining (- (Variables|Shield|GetShieldHealthRemaining) Damage)))
  (Development|PrintString (Utilities|String|ToString(Float) _output_get))
  (if (<= _output_get 0)
    (Shield|HandleShieldHealthDepleted)))

(fn HandleShieldHealthDepleted
  (FortWeapon|ModAmmo -1)
  (Class|BGATwinStretchCornBlockVolume|OnShieldPieceBroken (Variables|Shield|GetSpawnedShield))
  (if (> (FortWeapon|GetMagazineAmmoCount) 0)
    (Shield|ResetShieldHealth)))

(fn IdleFXParameters
  (params (Offset%20Position :type Vector) (Shield%20Rotation :type Rotator))
  (bind _spawnedidlefx (Variables|Default|GetSpawnedIdleFX)
    (:then
      (Effects|Components|ParticleSystem|SetVectorParameter _spawnedidlefx "Offset" Offset%20Position)
      (Effects|Components|ParticleSystem|SetVectorParameter _spawnedidlefx "Rotation" (Math|Rotator|GetRotationXVector Shield%20Rotation)))
    (:else)))

(fn NotifyChargingState
  (params (bIsCharging :type Boolean))
  (Default|CallOnChargingStateChanged bIsCharging))

(fn OnRep_NumQueuedProjectiles
  (Networking|CanExecuteCosmeticEvents
    (:True
      (Default|CallOnNumChargedChanged (Variables|Charging|GetNumQueuedProjectiles))
      (Variables|Default|GetSpawnedShards
        (:then
          (if (> (Variables|Charging|GetNumQueuedProjectiles) 0)
            (FX|UpdateChargedShardsFX)
            (else
              (Components|DestroyComponent (Variables|Default|GetSpawnedShards)))))
        (:else
          (if (> (Variables|Charging|GetNumQueuedProjectiles) 0)
            (bind _returnvalue (Niagara|SpawnSystemAttached (Variables|Charging|GetChargedShardsTemplate) (Variables|/Script/Engine.Actor|GetRootComponent (Game|GetInstigator)) "None" (Variables|Charging|GetQueuedFXOffset) "0, 0, 0" "KeepRelativeOffset" false true "AutoRelease"))
            (Variables|Default|SetSpawnedShards _returnvalue)
            (FX|UpdateChargedShardsFX)))))))

(fn OnRep_SpawnedShield)

(fn ResetShieldHealth
  (Variables|Shield|SetShieldHealthRemaining (ScalableFloat|GetValueatLevel (Variables|Shield|GetMaxShieldHealthPerShard))))

(fn SetIdleFXVisibility
  (params (Visible :type Boolean))
  (bind _spawnedidlefx (Variables|Default|GetSpawnedIdleFX)
    (:then
      (Utilities|FlowControl|Sequence
        (:then_0
          (if Visible
            (if (FortWeapon|IsEquipped)
              (Components|Activation|Activate _spawnedidlefx false)
              (else
                _))
            (Effects|Components|ParticleSystem|SetIntParameter _spawnedidlefx "Ammo Count" (FortWeapon|GetMagazineAmmoCount))
            (return _spawnedidlefx)
            (else
              (Components|Activation|Deactivate _spawnedidlefx))))
        (:then_1
          (if true
            (Effects|Components|ParticleSystem|SetBoolParameter _spawnedidlefx "PlayInitialize" false)
            (else
              (return 0))))
        (:then_2
          (Effects|Components|ParticleSystem|SetBoolParameter _spawnedidlefx "PlayPrj" false))))
    (:else)))

(fn SpawnShield
  (bind _returnvalue (Game|GetInstigator))
  (Development|PrintString "Spawning shield")
  ;[1 1 1] TODO PamF : We have to pool this at some point
  (bind _returnvalue_2 (Building|SpawnBuildingGameplayActor (Variables|Shield|GetShieldClass) (Math|Transform|MakeTransform "0, 0, 0") _returnvalue true))
  (Utilities|IsValid _returnvalue_2
    (:Is%20Valid
      (bind _asbga_twin_stretch_corn_block_volume (Utilities|Casting|CastToBGA_TwinStretchCorn_BlockVolume _returnvalue_2))
      (bind _output_get (Variables|Shield|SetSpawnedShield _asbga_twin_stretch_corn_block_volume))
      (Transformation|SetActorRelativeLocation _output_get (Class|FortPlayerViewFollowerComponent|GetRelativeOffsetfromPlayer (Actor|GetComponentbyClass _output_get "/Script/TwinStretchCornRuntime.FortPlayerViewFollowerComponent")) false false)
      (Transformation|AttachActorToActor _output_get _returnvalue "None" "KeepRelative" "SnapToTarget" "KeepWorld" true)
      (return (Variables|Shield|GetSpawnedShield)))))

(fn UpdateIdleFX
  (params (Ammo%20Count :type Integer) (Start%20Idle%20Animation :type Boolean) (Start%20Defense%20Animation :type Boolean) (Start%20Offense%20Animation :type Boolean) (StaticMeshComponent :type Static%20Mesh%20Component%20Object%20Reference%20%28by%20ref%29) (Deactivate%20Idle :type Boolean) (Unequipped :type Boolean))
  (bind _spawnedidlefx (Variables|Default|GetSpawnedIdleFX)
    (:then
      (if Deactivate%20Idle
        (if Unequipped
          (Components|Activation|Deactivate _spawnedidlefx)
          (else
            (Effects|Components|ParticleSystem|SetIntParameter _spawnedidlefx "Ammo Count" 0)))
        (else
          (bind _locations (Charging|GetCurrentShardLocations))
          (Niagara|NiagaraSetVectorArray _spawnedidlefx (Variables|Charging|GetShardLocationsParam) _locations)
          (Utilities|FlowControl|Sequence
            (:then_0
              (Effects|Components|ParticleSystem|SetIntParameter _spawnedidlefx "Ammo Count" Ammo%20Count))
            (:then_1
              ;[1 1 1] Idle
              (if Start%20Idle%20Animation
                (Effects|Components|ParticleSystem|SetBoolParameter _spawnedidlefx "Idle-Mode" true)
                (else
                  (Effects|Components|ParticleSystem|SetBoolParameter _spawnedidlefx "Idle-Mode" false))))
            (:then_2
              ;[1 1 1] Defense
              (if Start%20Defense%20Animation
                (Niagara|SetNiagaraStaticMeshComponent _spawnedidlefx "Static Mesh")
                (Effects|Components|ParticleSystem|SetBoolParameter _spawnedidlefx "DefenseMode" true)
                (else
                  (Effects|Components|ParticleSystem|SetBoolParameter _spawnedidlefx "DefenseMode" false))))
            (:then_3
              ;[1 1 1] Offense
              (if Start%20Offense%20Animation
                (Effects|Components|ParticleSystem|SetBoolParameter _spawnedidlefx "OffenseMode" true)
                (else
                  (Effects|Components|ParticleSystem|SetBoolParameter _spawnedidlefx "OffenseMode" false))))))))
    (:else)))

(fn UpdateChargedShardsFX
  (bind _spawnedshards (Variables|Default|GetSpawnedShards))
  (bind _locations (Charging|GetCurrentShardLocations))
  (Niagara|NiagaraSetVectorArray _spawnedshards (Variables|Charging|GetShardLocationsParam) _locations)
  (Effects|Components|ParticleSystem|SetIntParameter (Variables|Default|GetSpawnedShards) "LoopCount" (FortWeapon|GetMagazineAmmoCount))
  (bind _asfort_player_pawn (Utilities|Casting|CastToFortPlayerPawn (Game|GetInstigator))
    (:then
      (Niagara|SetNiagaraSkeletalMeshComponent _spawnedshards "Skeletal Mesh" (Variables|/Script/Engine.Character|GetMesh _asfort_player_pawn))
      (FX|UpdateIdleFX (Variables|Charging|GetNumQueuedProjectiles)))
    (:CastFailed
      (Development|PrintString "cast failed"))))

(fn ConstructionScript
  (|CallParentFunction))
